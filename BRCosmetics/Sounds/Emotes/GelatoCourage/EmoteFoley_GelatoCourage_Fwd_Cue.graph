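# Looping Sound Cue for the forward-movement foley of the GelatoCourage traversal emote.
# It mixes an idle loop, a forward loop wrapped in a volume fade-in envelope, and a tire-on-dirt loop.
# Each layer has a continuous modulator reading EmoteMoveSpeed for pitch and volume, with the idle layer's inverted volume range lowering it as speed rises.
# Emote foley attenuation, sound class and concurrency settings govern playback.
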SoundCue /BRCosmetics/Sounds/Emotes/GelatoCourage/EmoteFoley_GelatoCourage_Fwd_Cue
root=Mixer_1
nodes [
 {
  id=Enveloper_4
  children [WavePlayer_1]
  properties {pitchCurve {editorCurveData {keys [{time=0}]}}, volumeCurve {editorCurveData {keys [<at 0 0>,<at 1 1>]}}}
 }
 {id=Mixer_0, children [ModulatorContinuous_1, ModulatorContinuous_0], properties {inputVolume [0.5,1]}}
 {id=Mixer_1, children [ModulatorContinuous_2, Mixer_0], properties {inputVolume [1,1]}}
 {
  id=ModulatorContinuous_0
  children [Enveloper_4]
  properties {pitchModulationParams {minOutput=1, parameterName=EmoteMoveSpeed}, volumeModulationParams {parameterName=EmoteMoveSpeed}}
 }
 {
  id=ModulatorContinuous_1
  children [WavePlayer_11]
  properties {
   pitchModulationParams {minOutput=0.9, parameterName=EmoteMoveSpeed}
   volumeModulationParams {paramMode=MPM_Abs, parameterName=EmoteMoveSpeed}
  }
 }
 {
  id=ModulatorContinuous_2
  children [WavePlayer_0]
  properties {
   pitchModulationParams {maxOutput=1.2, minOutput=1.1, parameterName=EmoteMoveSpeed}
   volumeModulationParams {maxOutput=0.6, minOutput=0.8, parameterName=EmoteMoveSpeed}
  }
 }
 {id=WavePlayer_0, properties {bLooping=true, soundWaveAssetPtr=&/IP_Emote_GelatoCourage_Idle}}
 {id=WavePlayer_1, properties {bLooping=true, soundWaveAssetPtr=&/IP_Emote_GelatoCourage_Fwd_01}}
 {id=WavePlayer_11, properties {bLooping=true, soundWaveAssetPtr=&/IP_Emote_GelatoCourage_Tire_Dirt_LP}}
]
properties {
 attenuationSettings=&Fort_Emotes_Athena_Att
 audioPropertiesSheet=/BRCosmetics/Sounds/Settings/PropertySheets/Emotes/APS_Cosmetic_Emote_Foley
 maxDistance=1600
 soundClassObject=/Game/Sounds/FortSoundClassesAndModes/Fort_Foley_Emote
 volumeMultiplier=1
 concurrencyOverrides {maxCount=3}
 concurrencySet [/Game/Sounds/Concurrency_Rules/Athena_EmoteFoley_Conc]
 soundSubmixSends [{minSendLevel=0, soundSubmix=&Fortnite_AudioAnalysis_Music}]
}
