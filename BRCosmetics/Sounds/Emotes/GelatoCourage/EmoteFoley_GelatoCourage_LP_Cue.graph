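# Looping foley Sound Cue for the GelatoCourage emote, playing race background waves with emote attenuation, sound class and concurrency.
# Two layers each loop a random pick from their own six waves, starting after separate random delays, then pass through Doppler and slight random volume variation.
# The mixed layers get volume driven by the EmoteMoveSpeed parameter, with a non-zero floor, under an Enveloper that adds a short fade-in.
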SoundCue /BRCosmetics/Sounds/Emotes/GelatoCourage/EmoteFoley_GelatoCourage_LP_Cue
root=Enveloper_0
nodes [
 {id=Delay_0, children [Looping_0], properties {delayMax=1}}
 {id=Delay_2, children [Looping_2], properties {delayMax=1.5, delayMin=1}}
 {id=Doppler_0, children [Delay_2]}
 {id=Doppler_2, children [Delay_0]}
 {
  id=Enveloper_0
  children [ModulatorContinuous_0]
  properties {pitchCurve {editorCurveData {keys [{time=0}]}}, volumeCurve {editorCurveData {keys [<at 0 0>,<at 0.1 1>]}}}
 }
 {id=Looping_0, children [Random_0]}
 {id=Looping_2, children [Random_3]}
 {id=Mixer_1, children [Modulator_1,Modulator_0], properties {inputVolume [1,1]}}
 {
  id=ModulatorContinuous_0
  children [Mixer_1]
  properties {pitchModulationParams {minOutput=1, parameterName=EmoteMoveSpeed}, volumeModulationParams {minOutput=0.3, parameterName=EmoteMoveSpeed}}
 }
 {id=Modulator_0, children [Doppler_0], properties {volumeMax=1.02, volumeMin=0.98}}
 {id=Modulator_1, children [Doppler_2], properties {volumeMax=1.02, volumeMin=0.98}}
 {id=Random_0, children [WavePlayer_0,WavePlayer_1,WavePlayer_2,WavePlayer_3,WavePlayer_4,WavePlayer_5], properties {weights [1,1,1,1,1,1]}}
 {
  id=Random_3
  children [WavePlayer_6, WavePlayer_7, WavePlayer_8, WavePlayer_9, WavePlayer_10, WavePlayer_11]
  properties {weights [1,1,1,1,1,1]}
 }
 {id=WavePlayer_0, properties {soundWaveAssetPtr=&/IP_Emote_GelatoCourage_Race_BG_01}}
 {id=WavePlayer_1, properties {soundWaveAssetPtr=&/IP_Emote_GelatoCourage_Race_BG_02}}
 {id=WavePlayer_10, properties {soundWaveAssetPtr=&/IP_Emote_GelatoCourage_Race_BG_11}}
 {id=WavePlayer_11, properties {soundWaveAssetPtr=&/IP_Emote_GelatoCourage_Race_BG_12}}
 {id=WavePlayer_2, properties {soundWaveAssetPtr=&/IP_Emote_GelatoCourage_Race_BG_03}}
 {id=WavePlayer_3, properties {soundWaveAssetPtr=&/IP_Emote_GelatoCourage_Race_BG_04}}
 {id=WavePlayer_4, properties {soundWaveAssetPtr=&/IP_Emote_GelatoCourage_Race_BG_05}}
 {id=WavePlayer_5, properties {soundWaveAssetPtr=&/IP_Emote_GelatoCourage_Race_BG_06}}
 {id=WavePlayer_6, properties {soundWaveAssetPtr=&/IP_Emote_GelatoCourage_Race_BG_07}}
 {id=WavePlayer_7, properties {soundWaveAssetPtr=&/IP_Emote_GelatoCourage_Race_BG_08}}
 {id=WavePlayer_8, properties {soundWaveAssetPtr=&/IP_Emote_GelatoCourage_Race_BG_09}}
 {id=WavePlayer_9, properties {soundWaveAssetPtr=&/IP_Emote_GelatoCourage_Race_BG_10}}
]
properties {
 attenuationSettings=&Fort_Emotes_Athena_Att
 audioPropertiesSheet=/BRCosmetics/Sounds/Settings/PropertySheets/Emotes/APS_Cosmetic_Emote_Foley
 maxDistance=1600
 soundClassObject=/Game/Sounds/FortSoundClassesAndModes/Fort_Foley_Emote
 volumeMultiplier=0.65
 concurrencyOverrides {maxCount=3}
 concurrencySet [/Game/Sounds/Concurrency_Rules/Athena_EmoteFoley_Conc]
 soundSubmixSends [{minSendLevel=0, soundSubmix=&Fortnite_AudioAnalysis_Music}]
}
